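# Niagara impact effect for the Icicle Cane Bright pickaxe, mostly in icy blue tones.
# Flash_Elements layers spikes, star, flare, mesh and light, alongside smoke puffs and Sparks, whose spawn groups drive distinct motion and sizing.
# Debris takes velocity and gravity from User.PhysicalSurface, with CPU collision gated by platform quality, lifetime and age.
# Front-end modifiers adjust scale and offset, and colors pass through variant selectors.
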NiagaraSystem /BRCosmetics/Weapons/FORT_Melee/Pickaxe_IcicleCane_Bright/FX/NS_Pickaxe_IcicleCane_Bright_Impact
dependencies {
 usedDataInterfaces [
  &NiagaraDataInterfaceCurve
  &NiagaraDataInterfaceColorCurve
  &NiagaraDataInterfaceMeshRendererInfo
  &NiagaraDataInterfaceVector2DCurve
  &NiagaraDataInterfaceVectorCurve
  &NiagaraDataInterfacePlatformSet
 ]
 usedDynamicInputs [&FN_FrontEndVector3Modifier]
 usedModules [
  &self:SystemSpawnScript.SystemScriptSource.SystemScriptGraph.NiagaraNodeAssignment_0.SetVariables_#1
  &self:SystemSpawnScript.SystemScriptSource.SystemScriptGraph.NiagaraNodeAssignment_1.SetVariables_#2
  &SystemState
  &FN_VariantManager
 ]
}
emitters {
 FN_PhysMatDebris {
  emitterUpdateScript [
   &EmitterState()
   &SpawnBurst_Instantaneous(
    Spawn Count=&RandomRangeInt(Minimum=5, Maximum=10)
    Use Spawn Probability=false
    UseLoopCountLimit=false
   )
  ]
  particleSpawnScript [
   /SetVariables(Particles.BiasedRandom=&FloatFromCurve(
     FloatCurve=DataInterface() {propertyValues {Curve {keys [
         <at 0.2 1 RCIM_Cubic>
         {arriveTangent=-3.3394, interpMode=RCIM_Cubic, leaveTangent=-3.3394, tangentMode=RCTM_User, time=0.5, value=0.5}
         <at 0.8 0 RCIM_Cubic>
        ]}}}
     CurveIndex=&RandomRangeFloat(Minimum=0, Maximum=1)
    )) {isSetParameters=true}
   &InitializeParticle(
    Color Mode=::NewEnumerator0
    Lifetime=&RandomRangeFloat(Minimum=0.5, Maximum=1.25)
    Mass=&Lerp_Float(A=0.6, B=1.2, Alpha=$Particles.BiasedRandom)
    Mass Mode=::NewEnumerator1
    Mesh Scale=&Multiply_VectorByFloat(
     Vector=&RandomRangeVector(Minimum=<0.9 0.4 0.5>, Maximum=<1 1 1>)
     Float=&Lerp_Float(A=0.7, B=1, Alpha=$Particles.BiasedRandom)
    )
    Mesh Scale Mode=::NewEnumerator4
    UsePositionOffset=false
    Write Material Random=true
   )
   &FN_PhysicalSurface(
    Dirt Mesh Count=2
    Glass Mesh Count=3
    Ice Mesh Count=2
    Metal Gravity=&RandomRangeFloat(Minimum=-350, Maximum=-1800)
    Metal Mesh Count=2
    Metal Velocity=&RandomRangeFloat(Minimum=80, Maximum=375)
    Metal Velocity Cone Angle=130
    Physical Surface=$User.PhysicalSurface
    Stone Gravity=&RandomRangeFloat(Minimum=-700, Maximum=-1250)
    Stone Mesh Count=2
    Stone Velocity=&RandomRangeFloat(Minimum=80, Maximum=550)
    Stone Velocity Cone Angle=120
    Wood Gravity=&RandomRangeFloat(Minimum=-450, Maximum=-900)
    Wood Mesh Count=2
    Wood Scale=1.5
    Wood Velocity=&RandomRangeFloat(Minimum=80, Maximum=600)
    Wood Velocity Cone Angle=130
   )
   &AddVelocityInCone(
    Cone Angle=$Output.FN_PhysicalSurface.SurfaceResponseConeAngle
    Module.Use Velocity Falloff On Cone Axis=false
    Velocity Strength=$Output.FN_PhysicalSurface.SurfaceResponseVelocity
   )
   &InitialMeshOrientation(Use Rotation Vector=false)
   &AddRotationalVelocity(Rotation Rate=&RandomVector(Vector Scale=&RandomRangeFloat(Minimum=1, Maximum=2)))
   &ApplyInitialForces()
   /SetVariables(Particles.GravityOverrideValue=$Output.FN_PhysicalSurface.GravityOverride) {isSetParameters=true}
  ]
  particleUpdateScript [
   &ParticleState()
   &GravityForce(Gravity=$Particles.GravityOverrideValue)
   &Collision(
    Advanced Aging Rate=1
    Age Colliding Particles=false
    CPU Collision Trace Channel=ECC_WorldStatic
    Collision Enabled=&Boolean_AndOperation(
     Bool1=&IsPlatformSetActive(New Platform Set=DataInterface() {propertyValues {Platforms {qualityLevelMask=-8}}})
     Bool2=&Boolean_AndOperation(
      Bool1=&SetBoolByFloatComparison(A=$Particles.Initial.Lifetime, B=1)
      Bool2=&SetBoolByFloatComparison(A=$Particles.Age, B=0.12)
     )
    )
    Control Roll On Collision=true
    Friction=1
    Friction During a Bounce=0.25
    Max CPU Trace Length=100
    Module.Control Rest State Via Surface Normal=false
    Module.EnableMaxCollisionCount=false
    Module.Randomize Collision Normal Vector=true
    Randomize Collision Normal=0.25
    ScaleRateOfRotation=false
   )
   &SolveForcesAndVelocity(Clamp Velocity=false, Limit Acceleration=false)
   &ScaleMeshSize(Scale Factor=&VectorFromFloat(Value=&FloatFromCurve(FloatCurve=DataInterface() {propertyValues {Curve {keys [<at 0.6 1 RCIM_Cubic>, {arriveTangent=-6.308, interpMode=RCIM_Cubic, leaveTangent=-6.308, tangentMode=RCTM_User, time=1, value=0}]}}})))
   &DynamicMaterialParameters(Module.Param0WriteEnabled=true, Module.Param1WriteEnabled=true, Module.Param2WriteEnabled=false, Module.Param3WriteEnabled=false)
  ]
  rendererClasses [&NiagaraSpriteRendererProperties, &NiagaraMeshRendererProperties]
  renderers [
   {rendererClass=&NiagaraSpriteRendererProperties}
   {rendererClass=&NiagaraMeshRendererProperties, rendererIndex=1}
   {rendererClass=&NiagaraMeshRendererProperties, rendererIndex=2}
   {rendererClass=&NiagaraMeshRendererProperties, rendererIndex=3}
   {rendererClass=&NiagaraMeshRendererProperties, rendererIndex=4}
   {rendererClass=&NiagaraMeshRendererProperties, rendererIndex=5}
   {rendererClass=&NiagaraMeshRendererProperties, rendererIndex=6}
   {rendererClass=&NiagaraMeshRendererProperties, rendererIndex=7}
   {rendererClass=&NiagaraMeshRendererProperties, rendererIndex=8}
  ]
 }
 Flash_Elements {
  emitterUpdateScript [
   &EmitterState()
   &SpawnBurst_Instantaneous(
    Spawn Count=&RandomRangeInt(Minimum=6, Maximum=8)
    Use Spawn Probability=false
    UseLoopCountLimit=false
   )
   &SpawnBurst_Instantaneous(Spawn Group=1, Use Spawn Probability=false, UseLoopCountLimit=false) {module=SpawnBurst_Instantaneous001}
   &SpawnBurst_Instantaneous(Spawn Group=2, Use Spawn Probability=false, UseLoopCountLimit=false) {module=SpawnBurst_Instantaneous002}
  ]
  particleSpawnScript [
   &InitializeParticle(
    Color=&MaskLinearColor(Mask Source=1, Mask Parameter=$Emitter.SpawnGroup, Mask=2, Masked Linear Color=<rgba 1 1 1 0.6>, Passthrough Linear Color=<rgba 1 1 1 1>)
    Lifetime=&MaskFloat(Mask Source=1, Mask Parameter=$Emitter.SpawnGroup, Mask=2, Masked Float=0.6, Passthrough Float=0.25)
    Position Offset=$System.FEImpactOffset
    Position Offset Coordinate Space=::Local
    UsePositionOffset=true
    Write Material Random=true
   )
   &ShapeLocation(MaskSpawnGroup=false, Override Seed=false, Sphere Radius=1, UseOwnerScale=true)
   &SpriteFacingAndAlignment(Sprite Alignment=$Particles.ShapeLocation.ShapeNormal, WriteAlignment=true, WriteFacing=false)
   &FN_PerParticleLOD(Enable SpriteSize Distance Scaling=true, Sprite Size Scale Far=<2.5 2.5>)
   /SetVariables(Particles.SpriteSize=<1 1>, Particles.VisibilityTag=&SpawnGroup()) {isSetParameters=true}
   /SetVariables(
    Particles.Flash Spike Color:LinearColor=<rgba 0.3 0.4167 1 1>
    Particles.Flash Spike Lifetime:NiagaraFloat=&RandomRangeFloat(Minimum=0.05, Maximum=0.07)
    Particles.Flash Spike Sprite Size:Vector2f=&FN_FrontEndVector2Modifier(
     Front End:Vector2f=&RandomRangeVector2D(Minimum=<2 25>, Maximum=<7 45>)
     Gameplay:Vector2f=&RandomRangeVector2D(Minimum=<3 55>, Maximum=<10 70>)
    )
   ) {isSetParameters=true}
   /SetVariables(
    Particles.Flash Star Cam Offset=15
    Particles.Flash Star Lifetime=0.15
    Particles.Flash Star Rotation:NiagaraFloat=&RandomRangeFloat(Minimum=0, Maximum=360)
    Particles.Flash Star Sprite Size:Vector2f=&Multiply_Vector2D_ByFloat(
     Float=&Multiply_Float(A=$System.FrontEndScale, B=&RandomRangeFloat(Minimum=120, Maximum=150))
     Vector2D=$Particles.SpriteSize
    )
    Particles.Flash Star Sprite Size Alt:Vector2f=&Multiply_Vector2D_ByFloat(
     Float=&Multiply_Float(
      A=1
      B=&Multiply_Float(A=$System.FrontEndScale, B=&RandomRangeFloat(Minimum=120, Maximum=150))
     )
     Vector2D=$Particles.SpriteSize
    )
    Particles.Flash Star SubUV Index:NiagaraFloat=&RandomRangeFloat(Minimum=0, Maximum=3)
    Particles.Flash Star SubUV Index Alt:NiagaraFloat=&RandomRangeFloat(Minimum=0, Maximum=3)
   ) {isSetParameters=true}
   /SetVariables(
    Particles.Flash Flare Lifetime:NiagaraFloat=&FN_FrontEndFloatModifier(Gameplay=0.2, Front End=0.125)
    Particles.Flash Flare Sprite Size:Vector2f=&FN_FrontEndVector2Modifier(Front End:Vector2f=<90 90>, Gameplay:Vector2f=<190 190>)
   ) {isSetParameters=true}
   &DynamicMaterialParameters(
    Index 0 Param 1=0.2
    Index 0 Param 2=1
    Module.Index0Param0Write=true
    Module.Index0Param1Write=true
    Module.Index0Param2Write=true
    Module.Index0Param3Write=true
   )
   &InitialMeshOrientation(Mesh Orientation Mode=::NewEnumerator7, Rotation=<0.18 0 0>, Use Rotation Vector=true)
   /SetVariables(Particles.CameraOffsetAlt=20) {isSetParameters=true}
  ]
  particleUpdateScript [
   &ParticleState()
   /SetVariables(
    Particles.Update Flash Flare Age:NiagaraFloat=&Divide_Float(A=$Particles.Age, B=$Particles.Initial.Flash Flare Lifetime)
    Particles.Update Flash Spike Age:NiagaraFloat=&Divide_Float(A=$Particles.Age, B=$Particles.Initial.Flash Spike Lifetime)
    Particles.Update Flash Star Age:NiagaraFloat=&Divide_Float(A=$Particles.Age, B=$Particles.Initial.Flash Star Lifetime)
   ) {isSetParameters=true}
   /SetVariables(
    Particles.Update Flash Spike Color:LinearColor=&MakeLinearColorFromVectorAndFloat(
     "Vector (RGB)"=&MakeVectorFromLinearColorRGB(Color=$Particles.Initial.Flash Spike Color)
     "Float (Alpha)"=&FloatFromCurve(FloatCurve=DataInterface() {propertyValues {Curve {keys [<at 0.7 1>,<at 1 0>]}}}, CurveIndex=$Particles.Update Flash Spike Age)
    )
    Particles.Update Flash Spike Sprite Size:Vector2f=&Multiply_Vector2D_ByFloat(
     Float=&FloatFromCurve(FloatCurve=DataInterface() {propertyValues {Curve {keys [<at 0 0.6>,<at 1 2>]}}}, CurveIndex=$Particles.Update Flash Spike Age)
     Vector2D=$Particles.Initial.Flash Spike Sprite Size
    )
    Particles.Update Flash Star Alt Color:LinearColor=&VariantColorSelector(
     Variant Amount=0
     Original Color=&ColorFromCurve(
      ColorCurve=DataInterface() {propertyValues {
        AlphaCurve {keys [<at 0 1>,<at 0.8992 0>]}
        BlueCurve {keys [<at 0.1978 3>,<at 0.6127 1.5>]}
        GreenCurve {keys [<at 0.1978 2.3347>,<at 0.6127 0.525>]}
        RedCurve {keys [<at 0.1978 0.339>,<at 0.6127 0.6062>]}
       }}
      CurveIndex=$Particles.Update Flash Star Age
     )
     Variant 1 Color=&ColorFromCurve(
      ColorCurve=DataInterface() {propertyValues {
        AlphaCurve {keys [<at 0 1>,<at 0.8992 0>]}
        BlueCurve {keys [<at 0.1978 3>,<at 0.6127 1.5>]}
        GreenCurve {keys [<at 0.1978 2.3347>,<at 0.6127 0.4894>]}
        RedCurve {keys [<at 0.1978 0.339>,<at 0.6127 0.3975>]}
       }}
      CurveIndex=$Particles.Update Flash Star Age
     )
     Variant 2 Color=?
    )
    Particles.Update Flash Star Color:LinearColor=&VariantColorSelector(
     Variant Amount=0
     Original Color=&ColorFromCurve(
      ColorCurve=DataInterface() {propertyValues {
        AlphaCurve {keys [<at 0 1>,<at 0.8992 0>]}
        BlueCurve {keys [<at 0.1978 3>,<at 0.6127 1.5>]}
        GreenCurve {keys [<at 0.1978 2.7>,<at 0.6127 0.78>]}
        RedCurve {keys [<at 0.1978 1.8>,<at 0.6127 0.6>]}
       }}
      CurveIndex=$Particles.Update Flash Star Age
     )
     Variant 1 Color=&ColorFromCurve(
      ColorCurve=DataInterface() {propertyValues {
        AlphaCurve {keys [<at 0 1>,<at 0.8992 0>]}
        BlueCurve {keys [<at 0.1978 3>,<at 0.6127 1.5>]}
        GreenCurve {keys [<at 0.1978 2.55>,<at 0.6127 0.8>]}
        RedCurve {keys [<at 0.1978 1.2>,<at 0.6127 0.3>]}
       }}
      CurveIndex=$Particles.Update Flash Star Age
     )
     Variant 2 Color=?
    )
    Particles.Update Flash Star Sprize Size:Vector2f=&Multiply_Vector2D_ByFloat(
     Float=&FloatFromCurve(
      FloatCurve=DataInterface() {propertyValues {Curve {keys [
          {arriveTangent=2.7013, interpMode=RCIM_Cubic, leaveTangent=2.7013, tangentMode=RCTM_User, time=0, value=0}
          {interpMode=RCIM_Cubic, tangentMode=RCTM_User, time=1, value=1}
         ]}}}
      CurveIndex=$Particles.Update Flash Star Age
     )
     Vector2D=$Particles.Initial.Flash Star Sprite Size
    )
    Particles.Update Flash Star Sprize Size Alt:Vector2f=&Multiply_Vector2D_ByFloat(
     Float=&FloatFromCurve(
      FloatCurve=DataInterface() {propertyValues {Curve {keys [
          {arriveTangent=2.7013, interpMode=RCIM_Cubic, leaveTangent=2.7013, tangentMode=RCTM_User, time=0, value=0}
          {interpMode=RCIM_Cubic, tangentMode=RCTM_User, time=1, value=1}
         ]}}}
      CurveIndex=$Particles.Update Flash Star Age
     )
     Vector2D=$Particles.Initial.Flash Star Sprite Size Alt
    )
   ) {isSetParameters=true}
   /SetVariables(Particles.Update Flash Flare Color:LinearColor=&MakeLinearColorFromVectorAndFloat(
     "Vector (RGB)"=&Multiply_Vector(
      A=/BRCosmetics/Weapons/FORT_Melee/Pickaxe_GateHound/FX/VariantManager/VariantVectorSelector(
       Original Vector:Vector3f=&MakeVectorFromLinearColorRGB(Color=&ColorFromCurve(ColorCurve=DataInterface() {propertyValues {
           AlphaCurve {keys [<at 0 1 RCIM_Cubic>]}
           BlueCurve {keys [<at 0 3 RCIM_Cubic>,<at 0.1334 1.5>,<at 0.5918 1 RCIM_Cubic>]}
           GreenCurve {keys [<at 0 2.8875 RCIM_Cubic>,<at 0.1334 1.2938>,<at 0.5918 0.6 RCIM_Cubic>]}
           RedCurve {keys [<at 0 1.65 RCIM_Cubic>,<at 0.1334 0.675>,<at 0.5918 0.4 RCIM_Cubic>]}
          }}))
       Variant 1 Vector:Vector3f=&MakeVectorFromLinearColorRGB(Color=&ColorFromCurve(ColorCurve=DataInterface() {propertyValues {
           AlphaCurve {keys [<at 0 1 RCIM_Cubic>]}
           BlueCurve {keys [<at 0 3 RCIM_Cubic>,<at 0.1334 1.5>,<at 0.5918 1 RCIM_Cubic>]}
           GreenCurve {keys [<at 0 2.8875 RCIM_Cubic>,<at 0.1334 1.2>,<at 0.5918 0.65 RCIM_Cubic>]}
           RedCurve {keys [<at 0 1.65 RCIM_Cubic>,<at 0.1334 0.3>,<at 0.5918 0.3 RCIM_Cubic>]}
          }}))
       Variant 2 Vector:Vector3f=?
       Variant 3 Vector:Vector3f=?
       Variant Amount:NiagaraInt32=0
      )
      B=<1 1 1>
     )
     "Float (Alpha)"=&FloatFromCurve(FloatCurve=DataInterface() {propertyValues {Curve {keys [<at 0 1>,<at 1 0>]}}}, CurveIndex=$Particles.Update Flash Flare Age)
    )) {isSetParameters=true}
   &FN_Light_Attributes(
    Color=<rgba 0.242 0.621 1 1>
    Create Color Attribute=true
    Create Exponent Attribute=true
    Create Position Attribute=true
    Create Radius Attribute=true
    Exponent=5
    Position=<15 0 0>
    Radius=90
    Scale Color=&FloatFromCurve(
     FloatCurve=DataInterface() {propertyValues {Curve {keys [
         {interpMode=RCIM_Cubic, tangentMode=RCTM_User, time=0, value=1}
         {arriveTangent=-0.5426, interpMode=RCIM_Cubic, leaveTangent=-0.5426, tangentMode=RCTM_User, time=1, value=0}
        ]}}}
     Scale Curve=10
    )
   )
   &FN_Sprite_Attributes(
    Camera Offset:NiagaraFloat=&FN_FrontEndFloatModifier(Gameplay=15, Front End=15)
    Color=&MakeLinearColorFromVectorAndFloat(
     "Vector (RGB)"=&MakeVectorFromLinearColorRGB(Color=<rgba 1.5 2.5 3 1>)
     "Float (Alpha)"=&Multiply_Float(
      A=&FloatFromCurve(FloatCurve=DataInterface() {propertyValues {Curve {keys [
           {arriveTangent=9.0725, interpMode=RCIM_Cubic, leaveTangent=9.0725, tangentMode=RCTM_User, time=0, value=0}
           {arriveTangent=-2.5249, interpMode=RCIM_Cubic, leaveTangent=-2.5249, tangentMode=RCTM_User, time=0.1408, value=0.9847}
           {arriveTangent=-0.0405, interpMode=RCIM_Cubic, leaveTangent=-0.0405, tangentMode=RCTM_User, time=1, value=0}
          ]}}})
      B=&FNTimeOfDayScalar(Night Scale=0.3, Morning Scale=0.5, Evening Scale=0.5)
     )
    )
    Create Camera Offset Attribute=true
    Create Color Attribute=true
    Create SpriteSize Attribute=true
    SpriteSize=&Vector2DFromFloat(Value=&FloatFromCurve(
      FloatCurve=DataInterface() {propertyValues {Curve {keys [{arriveTangent=1.4962, interpMode=RCIM_Cubic, leaveTangent=1.4962, tangentMode=RCTM_User, time=0, value=0.7}, <at 1 1.25 RCIM_Cubic>]}}}
      Scale Curve=&Multiply_Float(A=120, B=$System.FrontEndScale)
     ))
    Write to Module Namespace=true
   ) {module=FN_Sprite_Attributes001}
   /SetVariables(Particles.DynamicParameterAlt:Vector4f=&MakeVector4(X=0.6, Y=1, Z=1, W=1)) {isSetParameters=true}
   &FN_Mesh_Attributes(
    Color=&MakeLinearColorFromVectorAndFloat(
     "Vector (RGB)"=&MakeVectorFromLinearColorRGB(Color=&VariantColorSelector(
       Variant Amount=0
       Original Color=&ColorFromCurve(ColorCurve=DataInterface() {propertyValues {
          AlphaCurve {keys [<at 0 1 RCIM_Cubic>]}
          BlueCurve {keys [<at 0 1 RCIM_Cubic>,<at 0.2955 1>,<at 0.7958 1 RCIM_Cubic>]}
          GreenCurve {keys [<at 0 0.875 RCIM_Cubic>,<at 0.2955 0.5667>,<at 0.7958 0.25 RCIM_Cubic>]}
          RedCurve {keys [<at 0 0.5 RCIM_Cubic>,<at 0.2955 0.35>,<at 0.7958 0.3125 RCIM_Cubic>]}
         }})
       Variant 1 Color=&ColorFromCurve(ColorCurve=DataInterface() {propertyValues {
          AlphaCurve {keys [<at 0 1 RCIM_Cubic>]}
          BlueCurve {keys [<at 0 1 RCIM_Cubic>,<at 0.2955 1>,<at 0.7958 1 RCIM_Cubic>]}
          GreenCurve {keys [<at 0 0.7747 RCIM_Cubic>,<at 0.2955 0.6125>,<at 0.7958 0.25 RCIM_Cubic>]}
          RedCurve {keys [<at 0 0.205 RCIM_Cubic>,<at 0.2955 0.225>,<at 0.7958 0.25 RCIM_Cubic>]}
         }})
       Variant 2 Color=?
      ))
     "Float (Alpha)"=&FloatFromCurve(FloatCurve=DataInterface() {propertyValues {Curve {keys [
          <at 0 0 RCIM_Cubic>
          {arriveTangentWeight=0.0581, interpMode=RCIM_Cubic, leaveTangentWeight=0.3488, tangentMode=RCTM_User, time=0.1, value=1}
          {interpMode=RCIM_Cubic, leaveTangent=-2.0698, time=1, value=0}
         ]}}})
    )
    Create Color Attribute=true
    Create Scale Attribute=true
    Scale=&Multiply_Vector(
     A=&VectorFromFloat(Value=1.2)
     B=&VectorFromFloat(Value=&FloatFromCurve(
       FloatCurve=DataInterface() {propertyValues {Curve {keys [
           {arriveTangent=0.7538, interpMode=RCIM_Cubic, leaveTangent=0.7538, leaveTangentWeight=0.4929, tangentMode=RCTM_User, time=0, value=1}
           {arriveTangentWeight=0.3333, interpMode=RCIM_Cubic, tangentMode=RCTM_User, time=1, value=1.3}
          ]}}}
       Scale Curve=$System.FrontEndScale
      ))
    )
    Write to Module Namespace=true
   )
   /SetVariables(Particles.DynamicParametersAlt:Vector4f=&MakeVector4(
     X=0
     Y=&FloatFromCurve(FloatCurve=DataInterface() {propertyValues {Curve {keys [
          {arriveTangent=2.6595, interpMode=RCIM_Cubic, leaveTangent=2.6595, leaveTangentWeight=0.1059, tangentMode=RCTM_User, time=0.115, value=0.55}
          {arriveTangent=0.8293, tangentMode=RCTM_User, time=1, value=2}
         ]}}})
     Z=0
     W=0
    )) {isSetParameters=true}
  ]
  rendererClasses [
   &NiagaraSpriteRendererProperties
   &NiagaraLightRendererProperties
   &NiagaraMeshRendererProperties
  ]
  renderers [
   {rendererClass=&NiagaraSpriteRendererProperties}
   {rendererClass=&NiagaraSpriteRendererProperties, rendererIndex=1}
   {rendererClass=&NiagaraSpriteRendererProperties, rendererIndex=2}
   {rendererClass=&NiagaraSpriteRendererProperties, rendererIndex=3}
   {rendererClass=&NiagaraLightRendererProperties, rendererIndex=4}
   {rendererClass=&NiagaraSpriteRendererProperties, rendererIndex=5}
   {rendererClass=&NiagaraMeshRendererProperties, rendererIndex=6}
  ]
 }
 Impact_Smoke {
  emitterUpdateScript [
   &EmitterState()
   &SpawnBurst_Instantaneous(
    Spawn Count=&RandomRangeInt(Minimum=2, Maximum=3)
    Use Spawn Probability=false
    UseLoopCountLimit=false
   )
  ]
  particleSpawnScript [
   &InitializeParticle(
    Color Maximum=<rgba 0.76 0.78 1 0.2>
    Color Minimum=<rgba 0.8 0.9333 1 0.15>
    Color Mode=::NewEnumerator2
    Lifetime Max=0.4
    Lifetime Min=0.3
    Lifetime Mode=::NewEnumerator1
    Position Offset=$System.FEImpactOffset
    Position Offset Coordinate Space=::Local
    Sprite Size Mode=::NewEnumerator3
    Uniform Sprite Size=&Multiply_Float(
     A=&FN_FrontEndFloatModifier(Gameplay=1, Front End=0.55)
     B=&RandomRangeFloat(Minimum=100, Maximum=130)
    )
    UsePositionOffset=true
    Write Material Random=true
   )
   &CameraOffset(Camera Offset Amount=&RandomRangeFloat(Minimum=40, Maximum=50, Evaluation Type=::NewEnumerator1))
   &ShapeLocation(
    Hemisphere Distribution=<1 0.5>
    MaskSpawnGroup=false
    Override Seed=false
    Rotation Angle=90
    Rotation Mode=::NewEnumerator1
    Sphere Radius=&MakeCustomFloatFromBool(Bool=$User.bIsFrontEnd, True Float=5, False Float=10)
    UseOwnerScale=true
   )
   &AddVelocity(Velocity=&Multiply_VectorByFloat(Vector=$Particles.ShapeLocation.ShapeVector, Float=10))
  ]
  particleUpdateScript [
   &ParticleState()
   &ScaleSpriteSize(Uniform Curve Sprite Scale=DataInterface() {propertyValues {Curve {keys [
        {arriveTangent=1.7761, interpMode=RCIM_Cubic, leaveTangent=1.7761, tangentMode=RCTM_User, time=0, value=0}
        {interpMode=RCIM_Cubic, tangentMode=RCTM_User, time=1, value=1}
       ]}}})
   &ScaleColor(
    Scale Alpha=&FloatFromCurve(FloatCurve=DataInterface() {propertyValues {Curve {keys [
         {arriveTangent=2.4358, interpMode=RCIM_Cubic, leaveTangent=2.4358, tangentMode=RCTM_User, time=0, value=0}
         <at 0.25 0.5 RCIM_Cubic>
         <at 1 0 RCIM_Cubic>
        ]}}})
    Scale RGB=&VectorFromCurve(VectorCurve=DataInterface() {propertyValues {XCurve {keys [<at 0 1>,<at 1 1>]}, YCurve {keys [<at 0 1>,<at 1 1>]}, ZCurve {keys [<at 0 1>,<at 1 1>]}}})
    ScaleA=true
    ScaleRGB=true
   )
   &DynamicMaterialParameters(
    Index 0 Param 1=&Add_Float(
     A=&RandomRangeFloat(Minimum=0, Maximum=1)
     B=&FloatFromCurve(
      FloatCurve=DataInterface() {propertyValues {Curve {keys [
          <at 0 0 RCIM_Cubic>
          {arriveTangent=1.5947, interpMode=RCIM_Cubic, leaveTangent=1.5947, tangentMode=RCTM_User, time=0.2557, value=0.5257}
          <at 1 1 RCIM_Cubic>
         ]}}}
      CurveIndex=$Particles.Age
      Scale Curve=0.25
     )
    )
    Module.Param0WriteEnabled=true
    Module.Param1WriteEnabled=true
    Module.Param2WriteEnabled=true
    Module.Param3WriteEnabled=true
   ) {module=DynamicMaterialParameters001}
   &SolveForcesAndVelocity(Clamp Velocity=false, Limit Acceleration=false)
  ]
  rendererClasses [&NiagaraSpriteRendererProperties]
  renderers [
   {rendererClass=&NiagaraSpriteRendererProperties}
   {rendererClass=&NiagaraSpriteRendererProperties, rendererIndex=1}
  ]
 }
 Sparks {
  emitterUpdateScript [
   &EmitterState(Life Cycle Mode=::NewEnumerator1, Loop Behavior=::NewEnumerator1, Loop Delay=0.05, Loop Duration=0.3, UseLoopDelay=true)
   &SpawnBurst_Instantaneous(Spawn Count=&UniformRangedInt(Minimum=6, Maximum=10), Use Spawn Probability=false)
   &SpawnBurst_Instantaneous(Spawn Count=&UniformRangedInt(Minimum=4, Maximum=7), Spawn Group=1, Use Spawn Probability=false) {module=SpawnBurst_Instantaneous002}
   &SpawnBurst_Instantaneous(
    Spawn Count=&RandomRangeInt(Minimum=8, Maximum=14)
    Spawn Group=2
    Use Spawn Probability=false
    UseLoopCountLimit=false
   ) {module=SpawnBurst_Instantaneous003}
   &SpawnBurst_Instantaneous(Spawn Count=&UniformRangedInt(Minimum=4, Maximum=7), Spawn Group=3, Use Spawn Probability=false) {module=SpawnBurst_Instantaneous001}
   /SetVariables(
    Emitter.PetalInnerColor:LinearColor=&VariantColorSelector(Variant Amount=0, Original Color=<rgba 0.4 0.55 1 1>, Variant 1 Color=<rgba 0.1878 0.5113 0.8348 1>, Variant 2 Color=?)
    Emitter.PetalMiddleColor:LinearColor=&VariantColorSelector(Variant Amount=0, Original Color=<rgba 0.65 0.9125 1 1>, Variant 1 Color=<rgba 0.3043 0.7913 1 1>, Variant 2 Color=?)
    Emitter.PetalOuterColor:LinearColor=&VariantColorSelector(Variant Amount=0, Original Color=<rgba 0.3583 0.3 1 1>, Variant 1 Color=<rgba 0.4 0.4 1 1>, Variant 2 Color=?)
   ) {isSetParameters=true}
  ]
  particleSpawnScript [
   &InitializeParticle(
    Color=&MergeHelper_ColorSwitch(
     0=&VariantColorSelector(
      Variant Amount=0
      Original Color=&RandomRangeLinearColor(Minimum=<rgba 1.95 2.7375 3 1>, Maximum=<rgba 1.2 1.5 3 1>)
      Variant 1 Color=&RandomRangeLinearColor(Minimum=<rgba 1.2 2.46 3 1>, Maximum=<rgba 0.75 1.6875 3 1>)
      Variant 2 Color=?
     )
     1=&VariantColorSelector(
      Variant Amount=0
      Original Color=&RandomRangeLinearColor(Minimum=<rgba 1.2 2.55 3 1>, Maximum=<rgba 0.75 0.9375 3 1>)
      Variant 1 Color=&RandomRangeLinearColor(Minimum=<rgba 0.75 2.325 3 1>, Maximum=<rgba 0.75 1.6875 3 1>)
      Variant 2 Color=?
     )
     2=<rgba 1 1 1 1>
     3=&VariantColorSelector(Variant Amount=0, Original Color=<rgba 0.7 0.925 1 1>, Variant 1 Color=<rgba 0.4035 0.6134 0.807 1>, Variant 2 Color=?)
     Index=$Emitter.SpawnGroup
     Num=::NewEnumerator2
    )
    Lifetime=&MergeHelper_FloatSwitch(
     0=&RandomRangeFloat(Minimum=0.166, Maximum=0.45)
     1=&RandomRangeFloat(Minimum=0.55, Maximum=0.85)
     2=&RandomRangeFloat(Minimum=0.7, Maximum=1)
     3=&RandomRangeFloat(Minimum=0.4, Maximum=0.6)
     Index=$Emitter.SpawnGroup
     Num=::NewEnumerator2
    )
    Mesh Scale Mode=::NewEnumerator3
    Mesh Uniform Scale=&Multiply_Float(
     A=&RandomRangeFloat(Minimum=0.14, Maximum=0.225)
     B=&FN_FrontEndFloatModifier(Gameplay=1, Front End=0.8)
    )
    Position Offset=$System.FEImpactOffset
    Position Offset Coordinate Space=::Local
    Sprite Rotation Mode=::NewEnumerator2
    Sprite Size=&MergeHelper_Vector2Switch(
     0=&Vector2DFromFloat(Value=&Multiply_Float(
       A=&FN_FrontEndFloatModifier(Gameplay=1, Front End=0.8)
       B=&RandomRangeFloat(Minimum=2.5, Maximum=7)
      ))
     1=&Vector2DFromFloat(Value=&Multiply_Float(A=&RandomRangeFloat(Minimum=8.5, Maximum=17), B=$System.FrontEndScale))
     2=&Multiply_Vector2D_ByFloat(
      Float=&Multiply_Float(
       A=&RandomRangeFloat(Minimum=0.245, Maximum=0.695)
       B=&FN_FrontEndFloatModifier(Gameplay=1, Front End=0.8)
      )
      Vector2D=<18.5 7>
     )
     3=?
     4=?
     Index=$Emitter.SpawnGroup
     Num=::NewEnumerator1
    )
    Sprite Size Mode=::NewEnumerator4
    Sprite UV Mode=::NewEnumerator3
    UsePositionOffset=true
    Write Material Random=true
   )
   &ShapeLocation(MaskSpawnGroup=false, Override Seed=false, Sphere Radius=2, Sphere Surface Distribution=0.05, UseOwnerScale=true)
   &AddVelocity(
    Module.Use Falloff Distance=false
    Origin Offset=$System.FEImpactOffset
    Override Seed=false
    Velocity Mode=::NewEnumerator1
    Velocity Speed=&Multiply_Float(
     A=&MergeHelper_FloatSwitch(
      0=&RandomRangeFloat(Minimum=400, Maximum=825)
      1=&RandomRangeFloat(Minimum=300, Maximum=400)
      2=&RandomRangeFloat(Minimum=1000, Maximum=1500)
      3=&RandomRangeFloat(Minimum=335, Maximum=600)
      Index=$Emitter.SpawnGroup
      Num=::NewEnumerator2
     )
     B=$System.FrontEndScale
    )
   )
   &CameraOffset(Camera Offset Amount=&MergeHelper_FloatSwitch(0=20, 1=15, 2=15, 3=10, Index=$Emitter.SpawnGroup, Num=::NewEnumerator2))
   /SetVariables(Particles.VisibilityTag=$Emitter.SpawnGroup) {isSetParameters=true}
  ]
  particleUpdateScript [
   &ParticleState()
   &GravityForce(Gravity=&MergeHelper_VectorSwitch(
     0=<0 0 -666>
     1=&RandomRangeVector(Minimum=<0 0 -150>, Maximum=<0 0 -200>)
     2=&RandomRangeVector(Minimum=<0 0 -150>, Maximum=<0 0 -200>)
     3=&RandomRangeVector(Minimum=<0 0 -333>, Maximum=<0 0 -444>)
     Index=$Particles.VisibilityTag
     Num=::NewEnumerator2
    ))
   &Drag(
    Drag=&MergeHelper_FloatSwitch(
     0=&UniformRangedFloat(Minimum=3, Maximum=8)
     1=&UniformRangedFloat(Minimum=3, Maximum=8)
     2=&Multiply_Float(
      A=&UniformRangedFloat(Minimum=1.2, Maximum=1.5)
      B=&FloatFromCurve(
       FloatCurve=DataInterface() {propertyValues {Curve {keys [
           {arriveTangent=-5.3551, interpMode=RCIM_Cubic, leaveTangent=-5.3551, tangentMode=RCTM_User, time=0, value=0}
           {arriveTangent=-138.8873, interpMode=RCIM_Cubic, leaveTangent=-138.8873, tangentMode=RCTM_User, time=0.05, value=7}
           <at 0.2 0.5>
          ]}}}
       Scale Curve=4
      )
     )
     3=&UniformRangedFloat(Minimum=3, Maximum=8)
     Index=$Particles.VisibilityTag
     Num=::NewEnumerator2
    )
    Use Linear Drag=true
    Use Rotational Drag=false
   )
   &CurlNoiseForce(
    Bias Noise=false
    Noise Frequency=&MergeHelper_FloatSwitch(0=0, 1=30, 2=0, 3=0, Index=$Particles.VisibilityTag, Num=::NewEnumerator2)
    Noise Strength=&MergeHelper_FloatSwitch(
     0=0
     1=&RandomRangeFloat(Minimum=225, Maximum=300)
     2=0
     3=0
     Index=$Particles.VisibilityTag
     Num=::NewEnumerator2
    )
    Pan Noise=false
   )
   &SolveForcesAndVelocity(Clamp Velocity=false, Limit Acceleration=false)
   &SpriteRotationRate(Rotation Rate=&MergeHelper_FloatSwitch(
     0=0
     1=&Multiply_Float(
      A=&UniformAOrBFloat(A=-1, Evaluation Type=::NewEnumerator0, Recalculate Random Each Loop=true, Override Seed=false)
      B=120
     )
     2=&Multiply_Float(
      A=&UniformAOrBFloat(A=-1, Evaluation Type=::NewEnumerator0, Recalculate Random Each Loop=true, Override Seed=false)
      B=&UniformRangedFloat(Minimum=50, Maximum=200)
     )
     3=?
     Index=$Particles.VisibilityTag
     Num=::NewEnumerator1
    ))
   &ScaleSpriteSize(Scale Factor=&MergeHelper_Vector2Switch(
     0=&Vector2DFromCurve(
      Vector2Curve=DataInterface() {propertyValues {
        XCurve {keys [
          {arriveTangent=5.9728, interpMode=RCIM_Cubic, leaveTangent=5.9728, tangentMode=RCTM_User, time=0, value=0.1}
          {arriveTangent=-0.1, interpMode=RCIM_Cubic, leaveTangent=-0.1, tangentMode=RCTM_User, time=0.18, value=0.8}
          {interpMode=RCIM_Cubic, tangentMode=RCTM_User, time=1, value=0}
         ]}
        YCurve {keys [
          {arriveTangent=6.2715, interpMode=RCIM_Cubic, leaveTangent=6.2715, tangentMode=RCTM_User, time=0, value=0.1}
          {arriveTangent=-0.1001, interpMode=RCIM_Cubic, leaveTangent=-0.1001, tangentMode=RCTM_User, time=0.18, value=0.8}
          {interpMode=RCIM_Cubic, tangentMode=RCTM_User, time=0.9988, value=0}
         ]}
       }}
      ScaleCurve=<1 1>
     )
     1=&Vector2DFromCurve(
      Vector2Curve=DataInterface() {propertyValues {
        XCurve {keys [
          {interpMode=RCIM_Cubic, leaveTangentWeight=0.075, tangentMode=RCTM_User, time=0, value=0}
          {arriveTangent=-0.1, interpMode=RCIM_Cubic, leaveTangent=-0.1, tangentMode=RCTM_User, time=0.225, value=1}
          {arriveTangent=-1.8257, interpMode=RCIM_Cubic, leaveTangent=-1.8257, tangentMode=RCTM_User, time=1, value=0}
         ]}
        YCurve {keys [
          {interpMode=RCIM_Cubic, leaveTangentWeight=0.075, tangentMode=RCTM_User, time=0, value=0}
          {arriveTangent=-0.1001, interpMode=RCIM_Cubic, leaveTangent=-0.1001, tangentMode=RCTM_User, time=0.225, value=1}
          {arriveTangent=-1.8257, interpMode=RCIM_Cubic, leaveTangent=-1.8257, tangentMode=RCTM_User, time=0.9988, value=0}
         ]}
       }}
      ScaleCurve=<1 1>
     )
     2=&Lerp_Vector2(
      A=&Multiply_Vector2(
       A=&MakeVector2D(
        X=&Multiply_Float(
         A=1
         B=&Sine(
          Normalized Angle=$Emitter.Age
          Period=&Multiply_Float(A=0.7, B=&Add_Float(A=$Particles.Initial.MaterialRandom, B=0.65))
          Scale=1
          Bias=0
         )
        )
        Y=1
       )
       B=&Vector2DFromFloat(Value=&FloatFromCurve(FloatCurve=DataInterface() {propertyValues {Curve {keys [<at 0 1.1>,<at 0.75 1.1>,<at 1 0>]}}}))
      )
      B=<1 1>
      Alpha=&FloatFromCurve(FloatCurve=DataInterface() {propertyValues {Curve {keys [<at 0.0574 1>,<at 0.1545 0>]}}})
     )
     3=?
     4=?
     Index=$Particles.VisibilityTag
     Num=::NewEnumerator1
    ))
   &ScaleMeshSize(Scale Factor=&VectorFromFloat(Value=&FloatFromCurve(FloatCurve=DataInterface() {propertyValues {Curve {keys [<at 0 1 RCIM_Cubic>,<at 1 0 RCIM_Cubic>]}}}, Scale Curve=2)))
   &ScaleSpriteSizeBySpeed(
    Max Scale Factor=&MergeHelper_Vector2Switch(0=<0.5 8>, 1=<1 1>, 2=<1 1>, 3=?, 4=?, Index=$Particles.VisibilityTag, Num=::NewEnumerator1)
    Min Scale Factor=&MergeHelper_Vector2Switch(0=<0.5 1>, 1=<1 1>, 2=<1 1>, 3=?, 4=?, Index=$Particles.VisibilityTag, Num=::NewEnumerator1)
    Velocity Threshold=666
   )
   &ScaleColor(
    Scale Mode=::NewEnumerator1
    Scale RGBA=&MakeVector4FromVectorAndFloat(
     "Vector (RGB)"=&MergeHelper_VectorSwitch(
      0=&VectorFromFloat(Value=&FloatFromCurve(
        FloatCurve=DataInterface() {propertyValues {Curve {keys [
            {interpMode=RCIM_Cubic, tangentMode=RCTM_User, time=0, value=1}
            <at 0.3 0.2>
            {interpMode=RCIM_Cubic, tangentMode=RCTM_User, time=1, value=0}
           ]}}}
        Scale Curve=10
       ))
      1=<1 1 1>
      2=<1 1 1>
      3=<1 1 1>
      Index=$Particles.VisibilityTag
      Num=::NewEnumerator2
     )
     "Float (Alpha)"=1
    )
    ScaleRGBA=true
   )
   /SetVariables(
    Particles.GlowColor=&MakeLinearColorFromVectorAndFloat(
     "Vector (RGB)"=&MakeVectorFromLinearColorRGB(Color=$Particles.Color)
     "Float (Alpha)"=&FloatFromCurve(FloatCurve=DataInterface() {propertyValues {Curve {keys [<at 0.5 1>,<at 1 0>]}}}, Scale Curve=0.03)
    )
    Particles.GlowSize=&Multiply_Vector2D(A=$Particles.SpriteSize, B=<8 2>)
   ) {isSetParameters=true}
  ]
  rendererClasses [&NiagaraSpriteRendererProperties, &NiagaraMeshRendererProperties]
  renderers [
   {rendererClass=&NiagaraSpriteRendererProperties}
   {rendererClass=&NiagaraMeshRendererProperties, rendererIndex=1}
   {rendererClass=&NiagaraSpriteRendererProperties, rendererIndex=2}
   {rendererClass=&NiagaraSpriteRendererProperties, rendererIndex=3}
   {rendererClass=&NiagaraSpriteRendererProperties, rendererIndex=4}
  ]
 }
}
summary {
 systemName=NS_Pickaxe_IcicleCane_Bright_Impact
 emitters [
  {
   emitterName=Flash_Elements
   rendererClasses [
    &NiagaraSpriteRendererProperties
    &NiagaraLightRendererProperties
    &NiagaraMeshRendererProperties
   ]
  }
  {emitterName=Sparks, rendererClasses [&NiagaraSpriteRendererProperties, &NiagaraMeshRendererProperties]}
  {emitterName=Impact_Smoke, rendererClasses [&NiagaraSpriteRendererProperties]}
  {
   emitterName=FN_PhysMatDebris
   rendererClasses [&NiagaraSpriteRendererProperties, &NiagaraMeshRendererProperties]
  }
 ]
 userVariables [
  User.bIsFrontEnd {type=NiagaraBool}
  User.bIsFrontEndPreview {type=NiagaraBool}
  User.PhysicalSurface {type=NiagaraInt32}
  User.SwingSectionIndex {type=NiagaraInt32}
  User.Variant {type=NiagaraFloat, defaultValue=0, description=""}
 ]
}
