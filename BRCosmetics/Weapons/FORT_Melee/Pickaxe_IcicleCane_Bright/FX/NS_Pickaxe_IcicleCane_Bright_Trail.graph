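# Niagara swing trail for the bright Icicle Cane pickaxe, drawing icy blue ribbons and shedding particles as the weapon moves.
# MainTrail and thin TipTrail build anim-trail ribbons with color fades and dynamic material parameters.
# Particle spawns three distance-based groups at a socket, spaced by swing section, with inherited velocity, gravity, drag and sprite and mesh renderers.
# Shared system values and helpers handle show/hide, front-end scale and time-of-day brightness.
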NiagaraSystem /BRCosmetics/Weapons/FORT_Melee/Pickaxe_IcicleCane_Bright/FX/NS_Pickaxe_IcicleCane_Bright_Trail
dependencies {
 usedDataInterfaces [
  &NiagaraDataInterfaceCurve
  &NiagaraDataInterfaceColorCurve
  &NiagaraDataInterfaceSpriteRendererInfo
  &NiagaraDataInterfaceSocketReader
 ]
 usedDynamicInputs [
  &FN_FrontEndFloatModifier
  &Boolean_NotOperation
  &SetBoolByIntComparison
 ]
 usedModules [
  &self:SystemSpawnScript.SystemScriptSource.SystemScriptGraph.NiagaraNodeAssignment_1.SetVariables_#1
  &self:SystemSpawnScript.SystemScriptSource.SystemScriptGraph.NiagaraNodeAssignment_5.SetVariables_#2
  &self:SystemSpawnScript.SystemScriptSource.SystemScriptGraph.NiagaraNodeAssignment_0.SetVariables_#3
  &self:SystemSpawnScript.SystemScriptSource.SystemScriptGraph.NiagaraNodeAssignment_2.SetVariables_#4
  &SystemState
  &self:SystemSpawnScript.SystemScriptSource.SystemScriptGraph.NiagaraNodeAssignment_3.SetVariables_#5
  &self:SystemSpawnScript.SystemScriptSource.SystemScriptGraph.NiagaraNodeAssignment_4.SetVariables_#6
  &FN_VariantManager
 ]
}
emitters {
 MainTrail {
  emitterSpawnScript [&FN_RibbonIDOffsetUtility()]
  emitterUpdateScript [
   &EmitterState()
   &FN_SpawnAnimTrail()
  ]
  particleSpawnScript [
   &FN_InitializeAnimTrailBoneIndecies()
   &FN_InitializeAnimTrail(Lifetime=0.22)
  ]
  particleUpdateScript [
   &FN_UpdateAnimTrail() {module=FN_UpdateAnimTrail001}
   &ScaleColor(
    Scale Alpha=$System.SwingGeneralFade
    Scale RGB=&VectorFromFloat(Value=$System.ToDScalar)
    ScaleA=true
    ScaleRGB=true
   ) {module=ScaleColor001}
   &DynamicMaterialParameters(
    Index 0 Param 3=0.01
    Index 0 Param 4=1
    Index 1 Param 3=&FloatFromCurve(FloatCurve=DataInterface() {propertyValues {Curve {keys [<at 0 0>,<at 1 1>]}}}, CurveIndex=$Emitter.Age, Scale Curve=-1)
    Index 1 Param 4=0.4
    Index 2 Param 1=20
    Index 2 Param 2=2
    Index 2 Param 4=0.6
    Module.Index0Float0Write=true
    Module.Index0Float1Write=true
    Module.Index0Float2Write=true
    Module.Index0Float3Write=true
    Module.Index1Float0Write=true
    Module.Index1Float1Write=true
    Module.Index1Float2Write=true
    Module.Index1Float3Write=true
    Module.Index2Float0Write=true
    Module.Index2Float1Write=true
    Module.Index2Float2Write=false
    Module.Index2Float3Write=false
    Write Parameter Index 1=::NewEnumerator1
    Write Parameter Index 2=::NewEnumerator1
   )
   /SetVariables(
    Particles.DynamicParametersAlt:Vector4f=&MakeVector4(
     X=0.2
     Y=0
     Z=&FloatFromCurve(FloatCurve=DataInterface() {propertyValues {Curve {keys [<at 0 0>,<at 1 1>]}}}, CurveIndex=$Emitter.Age, Scale Curve=-1)
     W=0.22
    )
    Particles.DynamicParametersAlt2:Vector4f=&MakeVector4(X=0, Y=0, Z=0, W=1)
   ) {isSetParameters=true}
   &ScaleColor(
    Scale Alpha=3
    Scale RGB=&Multiply_Vector(
     A=/BRCosmetics/Weapons/FORT_Melee/Pickaxe_GateHound/FX/VariantManager/VariantVectorSelector(
      Original Vector:Vector3f=&MakeVectorFromLinearColorRGB(Color=&ColorFromCurve(ColorCurve=DataInterface() {propertyValues {
          AlphaCurve {keys []}
          BlueCurve {keys [<at 0 1.5 RCIM_Cubic>,<at 0.5683 0.6 RCIM_Cubic>]}
          GreenCurve {keys [<at 0 0.8812 RCIM_Cubic>,<at 0.5683 0.252 RCIM_Cubic>]}
          RedCurve {keys [<at 0 0.675 RCIM_Cubic>,<at 0.5683 0.24 RCIM_Cubic>]}
         }}))
      Variant 1 Vector:Vector3f=&MakeVectorFromLinearColorRGB(Color=&ColorFromCurve(ColorCurve=DataInterface() {propertyValues {
          AlphaCurve {keys []}
          BlueCurve {keys [<at 0 1.5 RCIM_Cubic>,<at 0.5683 0.6 RCIM_Cubic>]}
          GreenCurve {keys [<at 0 0.9572 RCIM_Cubic>,<at 0.5683 0.3504 RCIM_Cubic>]}
          RedCurve {keys [<at 0 0.3682 RCIM_Cubic>,<at 0.5683 0.1722 RCIM_Cubic>]}
         }}))
      Variant 2 Vector:Vector3f=?
      Variant 3 Vector:Vector3f=?
      Variant Amount:NiagaraInt32=0
     )
     B=<1 1 1>
    )
    ScaleA=true
    ScaleRGB=true
   )
  ]
  rendererClasses [&NiagaraRibbonRendererProperties]
  renderers [
   {rendererClass=&NiagaraRibbonRendererProperties}
   {rendererClass=&NiagaraRibbonRendererProperties, rendererIndex=1}
  ]
 }
 Particle {
  emitterUpdateScript [
   &EmitterState(Life Cycle Mode=::NewEnumerator1, Loop Behavior=::NewEnumerator1, Loop Duration=0.3, UseLoopDelay=false)
   &SpawnPerUnit(
    Max Movement Threshold=2500
    Module.Use Max Movement Threshold=false
    Module.Use Movement Tolerance=false
    Spawn Probability:NiagaraFloat=&Multiply_Float(A=0.85, B=&MakeCustomFloatFromBool(Bool=$System.ShowFX, True Float=1, False Float=0))
    Spawn Spacing:NiagaraFloat=&MergeHelper_FloatSwitch(0=15, 1=15, 2=23, 3=19, Index=$User.SwingSectionIndex, Num=::NewEnumerator2)
    Use Spawn Probability=true
   )
   &SpawnPerUnit(
    Max Movement Threshold=2500
    Module.Use Max Movement Threshold=false
    Module.Use Movement Tolerance=false
    Spawn Group=1
    Spawn Probability:NiagaraFloat=&Multiply_Float(A=0.85, B=&MakeCustomFloatFromBool(Bool=$System.ShowFX, True Float=1, False Float=0))
    Spawn Spacing:NiagaraFloat=&MergeHelper_FloatSwitch(0=20, 1=20, 2=28, 3=25, Index=$User.SwingSectionIndex, Num=::NewEnumerator2)
    Use Spawn Probability=true
   ) {module=SpawnPerUnit001}
   &SpawnPerUnit(
    Max Movement Threshold=2500
    Module.Use Max Movement Threshold=false
    Module.Use Movement Tolerance=false
    Spawn Group=2
    Spawn Probability:NiagaraFloat=&MakeCustomFloatFromBool(Bool=$System.ShowFX, True Float=1, False Float=0)
    Spawn Spacing:NiagaraFloat=&MergeHelper_FloatSwitch(0=10, 1=10, 2=12, 3=15, Index=$User.SwingSectionIndex, Num=::NewEnumerator2)
    Use Spawn Probability=true
   ) {module=SpawnPerUnit002}
   /SetVariables(
    Emitter.PetalInnerColor:LinearColor=&VariantColorSelector(Variant Amount=0, Original Color=<rgba 0.4 0.55 1 1>, Variant 1 Color=<rgba 0.1878 0.5113 0.8348 1>, Variant 2 Color=?)
    Emitter.PetalMiddleColor:LinearColor=&VariantColorSelector(Variant Amount=0, Original Color=<rgba 0.65 0.9125 1 1>, Variant 1 Color=<rgba 0.3043 0.7913 1 1>, Variant 2 Color=?)
    Emitter.PetalOuterColor:LinearColor=&VariantColorSelector(Variant Amount=0, Original Color=<rgba 0.3583 0.3 1 1>, Variant 1 Color=<rgba 0.4 0.4 1 1>, Variant 2 Color=?)
   ) {isSetParameters=true}
  ]
  particleSpawnScript [
   &InitializeParticle(
    Color=&VariantColorSelector(Variant Amount=0, Original Color=<rgba 0.7 0.925 1 1>, Variant 1 Color=<rgba 0.4035 0.6134 0.807 1>, Variant 2 Color=?)
    Lifetime=&MergeHelper_FloatSwitch(
     0=&RandomRangeFloat(Minimum=0.7, Maximum=1)
     1=&RandomRangeFloat(Minimum=0.4, Maximum=0.65)
     2=&RandomRangeFloat(Minimum=0.45, Maximum=0.7)
     3=?
     Index=$Emitter.SpawnGroup
     Num=::NewEnumerator1
    )
    Mesh Scale Mode=::NewEnumerator3
    Mesh Uniform Scale=&Multiply_Float(A=&RandomRangeFloat(Minimum=0.1, Maximum=0.18), B=$System.FrontEndScale)
    Sprite Rotation Mode=::NewEnumerator2
    Sprite Size=&Multiply_Vector2D_ByFloat(
     Float=&Multiply_Float(A=&RandomRangeFloat(Minimum=0.245, Maximum=0.695), B=$System.FrontEndScale)
     Vector2D=<18.5 7>
    )
    Sprite Size Mode=::NewEnumerator4
    Sprite UV Mode=::NewEnumerator3
    UsePositionOffset=false
    Write Material Random=true
   )
   &SocketLocation(
    Interpolate Socket Sampling=true
    Orientation Sampling=::NewEnumerator1
    SamplePosOffset=false
    Socket Index=1
    Socket Reader=$System.Socket Reader
    Socket Sampling Mode=::NewEnumerator3
    VelocitySpeedLimit=false
   )
   &ShapeLocation(MaskSpawnGroup=false, Override Seed=false, Sphere Radius=20, UseOwnerScale=true)
   &InheritVelocity(
    Inherited Velocity Speed Limit:NiagaraFloat=&MergeHelper_FloatSwitch(
     0=&RandomRangeFloat(Minimum=200, Maximum=350)
     1=&RandomRangeFloat(Minimum=350, Maximum=600)
     2=&RandomRangeFloat(Minimum=80, Maximum=140)
     3=?
     Index=$Emitter.SpawnGroup
     Num=::NewEnumerator1
    )
    Speed Limit=true
    UseSpeedThreshold=false
    Velocity Source=$Output.SocketLocation.SampledVelocity
   )
   &CameraOffset(Camera Offset Amount=&MergeHelper_FloatSwitch(0=20, 1=10, 2=20, 3=?, Index=$Emitter.SpawnGroup, Num=::NewEnumerator1))
   /SetVariables(Particles.VisibilityTag=$Emitter.SpawnGroup) {isSetParameters=true}
  ]
  particleUpdateScript [
   &ParticleState()
   &ScaleMeshSize(Scale Factor=&VectorFromFloat(Value=&FloatFromCurve(FloatCurve=DataInterface() {propertyValues {Curve {keys [<at 0 1 RCIM_Cubic>,<at 1 0 RCIM_Cubic>]}}}, Scale Curve=2)))
   &Drag(
    Drag=&Multiply_Float(
     A=&UniformRangedFloat(Minimum=1.2, Maximum=1.5)
     B=&FloatFromCurve(
      FloatCurve=DataInterface() {propertyValues {Curve {keys [
          {arriveTangent=-5.3551, interpMode=RCIM_Cubic, leaveTangent=-5.3551, tangentMode=RCTM_User, time=0, value=0}
          {arriveTangent=-138.8873, interpMode=RCIM_Cubic, leaveTangent=-138.8873, tangentMode=RCTM_User, time=0.05, value=7}
          <at 0.2 0.5>
         ]}}}
      Scale Curve=4
     )
    )
    Use Linear Drag=true
    Use Rotational Drag=false
   )
   &SubUVAnimation(
    Output Alternate Renderer Binding=true
    "Sprite Renderer "=DataInterface() {propertyValues {SpriteRenderer=&self:Particle_0.NiagaraSpriteRendererProperties_1}}
    UseEndFrame=false
    UseStartFrame=false
   )
   &ScaleColor(
    Scale RGB=&VectorFromFloat(Value=&FloatFromCurve(
      FloatCurve=DataInterface() {propertyValues {Curve {keys [{arriveTangent=-2.7905, interpMode=RCIM_Cubic, leaveTangent=-2.7905, tangentMode=RCTM_User, time=0, value=2}, <at 1 1 RCIM_Cubic>]}}}
      Scale Curve=$System.ToDScalar
     ))
    ScaleA=true
    ScaleRGB=true
   )
   &GravityForce(Gravity=&RandomRangeVector(Minimum=<0 0 -150>, Maximum=<0 0 -200>))
   &CurlNoiseForce(
    Bias Noise=false
    Noise Frequency=20
    Noise Strength=&MergeHelper_FloatSwitch(
     0=&RandomRangeFloat(Minimum=225, Maximum=350)
     1=0
     2=0
     3=?
     Index=$Particles.VisibilityTag
     Num=::NewEnumerator1
    )
    Pan Noise=false
   )
   &ScaleSpriteSize(Scale Factor=&Lerp_Vector2(
     A=&Multiply_Vector2(
      A=&MakeVector2D(
       X=&Multiply_Float(
        A=1
        B=&Sine(
         Normalized Angle=$Emitter.Age
         Period=&Multiply_Float(A=0.7, B=&Add_Float(A=$Particles.Initial.MaterialRandom, B=0.65))
         Scale=1
         Bias=0
        )
       )
       Y=1
      )
      B=&Vector2DFromFloat(Value=&FloatFromCurve(FloatCurve=DataInterface() {propertyValues {Curve {keys [<at 0 1.1>,<at 0.75 1.1>,<at 1 0>]}}}))
     )
     B=<1 1>
     Alpha=&FloatFromCurve(FloatCurve=DataInterface() {propertyValues {Curve {keys [<at 0.0574 1>,<at 0.1545 0>]}}})
    ))
   &SpriteRotationRate(Rotation Rate=&Multiply_Float(
     A=&UniformAOrBFloat(A=-1, Evaluation Type=::NewEnumerator0, Recalculate Random Each Loop=true, Override Seed=false)
     B=&MergeHelper_FloatSwitch(
      0=&UniformRangedFloat(Minimum=50, Maximum=200)
      1=0
      2=0
      3=?
      Index=$Particles.VisibilityTag
      Num=::NewEnumerator1
     )
    ))
   &FN_Sprite_Attributes(
    Color=&MakeLinearColorFromVectorAndFloat(
     "Vector (RGB)"=&MakeVectorFromLinearColorRGB(Color=&RandomRangeLinearColor(Minimum=<rgba 0.2563 0.299 0.4271 0.6>, Maximum=<rgba 0.6 0.6667 1 0.8>))
     "Float (Alpha)"=&FloatFromCurve(
      FloatCurve=DataInterface() {propertyValues {Curve {keys [
          {arriveTangent=7.9556, interpMode=RCIM_Cubic, leaveTangent=7.9556, tangentMode=RCTM_User, time=0, value=0}
          {arriveTangentWeight=0.0667, interpMode=RCIM_Cubic, leaveTangentWeight=0.1, tangentMode=RCTM_User, time=0.2, value=1}
          {arriveTangentWeight=0.1601, interpMode=RCIM_Cubic, leaveTangentWeight=0.2668, tangentMode=RCTM_User, time=0.5, value=1}
          {arriveTangent=-3.5681, tangentMode=RCTM_User, time=1, value=0}
         ]}}}
      Scale Curve=&Multiply_Float(A=0.08, B=&FNTimeOfDayScalar(Night Scale=0.65, Morning Scale=0.675, Evening Scale=0.7))
     )
    )
    Create Color Attribute=true
    Create SpriteSize Attribute=true
    SpriteSize=&Multiply_Vector2D_ByFloat(
     Float=&MakeFloatFromVector2d(Vector 2D=&Vector2DFromFloat(Value=&FloatFromCurve(FloatCurve=DataInterface() {propertyValues {Curve {keys [
            {arriveTangent=0.4145, interpMode=RCIM_Cubic, leaveTangent=0.4145, tangentMode=RCTM_User, time=0, value=0.8}
            {arriveTangent=0.2495, interpMode=RCIM_Cubic, leaveTangent=0.2495, tangentMode=RCTM_User, time=0.3, value=0.9}
            {arriveTangentWeight=0.2333, interpMode=RCIM_Cubic, tangentMode=RCTM_User, time=1, value=1}
           ]}}})))
     Vector2D=&Multiply_Vector2D_ByFloat(
      Float=$System.FrontEndScale
      Vector2D=&Vector2DFromFloat(Value=&RandomRangeFloat(Minimum=70, Maximum=90))
     )
    )
    Write to Module Namespace=true
   )
   &SolveForcesAndVelocity(Clamp Velocity=false, Limit Acceleration=false)
  ]
  rendererClasses [&NiagaraSpriteRendererProperties, &NiagaraMeshRendererProperties]
  renderers [
   {rendererClass=&NiagaraSpriteRendererProperties}
   {rendererClass=&NiagaraMeshRendererProperties, rendererIndex=1}
   {rendererClass=&NiagaraSpriteRendererProperties, rendererIndex=2}
  ]
 }
 TipTrail {
  emitterSpawnScript [&FN_RibbonIDOffsetUtility()]
  emitterUpdateScript [
   &EmitterState()
   &FN_SpawnAnimTrail()
  ]
  particleSpawnScript [
   &FN_InitializeAnimTrailBoneIndecies(First Bone or Socket Index=1, Second Bone or Socket Index=2)
   &FN_InitializeAnimTrail(
    AnimTrailWidth=2
    Color=&MultiplyLinearColor(
     Linear Color=&VariantColorSelector(Variant Amount=0, Original Color=<rgba 1.5 2.25 3 1>, Variant 1 Color=<rgba 1.0957 2.3652 3 1>, Variant 2 Color=?)
     Scale Factor=&MakeVector4_FromLinearColor(LinearColor=&FNTimeOfDayColor(Night Color=<rgba 0.5417 0.5 1 1>, Morning Color=<rgba 0.7 0.825 1 1>, Day Color=<rgba 1 1 1 1>, Evening Color=<rgba 0.9 0.9417 1 1>))
    )
    Lifetime=0.225
   )
  ]
  particleUpdateScript [
   &FN_UpdateAnimTrail() {module=FN_UpdateAnimTrail001}
   &ScaleColor(
    Scale Alpha=$System.SwingGeneralFade
    Scale RGB=&VectorFromFloat(Value=&FloatFromCurve(
      FloatCurve=DataInterface() {propertyValues {Curve {keys [<at 0 3>,<at 0.2 1>,<at 1 0>]}}}
      Scale Curve=&FNTimeOfDayScalar(Night Scale=0.5, Morning Scale=0.7, Day Scale=0.935, Evening Scale=0.8)
     ))
    ScaleA=true
    ScaleRGB=true
   ) {module=ScaleColor001}
   &ScaleRibbonWidth(Ribbon Width Scale=&FloatFromCurve(FloatCurve=DataInterface() {propertyValues {Curve {keys [<at 0 1>, {leaveTangent=-1.7704, time=1, value=0}]}}}))
   &DynamicMaterialParameters(
    Index 0 Param 4=1
    Index 1 Param 3=&FloatFromCurve(FloatCurve=DataInterface() {propertyValues {Curve {keys [<at 0 0>,<at 1 1>]}}}, CurveIndex=$Emitter.Age, Scale Curve=-1)
    Index 1 Param 4=1
    Index 2 Param 1=40
    Index 2 Param 2=10
    Index 2 Param 4=0.6
    Module.Index0Float0Write=true
    Module.Index0Float1Write=true
    Module.Index0Float2Write=true
    Module.Index0Float3Write=true
    Module.Index1Float0Write=true
    Module.Index1Float1Write=true
    Module.Index1Float2Write=true
    Module.Index1Float3Write=true
    Module.Index2Float0Write=true
    Module.Index2Float1Write=true
    Module.Index2Float2Write=false
    Module.Index2Float3Write=false
    Write Parameter Index 1=::NewEnumerator1
    Write Parameter Index 2=::NewEnumerator1
   )
   &ScaleColor(
    Scale RGB=&VectorFromFloat(Value=&FloatFromCurve(FloatCurve=DataInterface() {propertyValues {Curve {keys [
          {arriveTangent=-20.3431, interpMode=RCIM_Cubic, leaveTangent=-20.3431, tangentMode=RCTM_User, time=0, value=5}
          {arriveTangent=-5.2915, interpMode=RCIM_Cubic, leaveTangent=-5.2915, tangentMode=RCTM_User, time=0.2, value=1.5}
          <at 1 0 RCIM_Cubic>
         ]}}}))
    ScaleA=true
    ScaleRGB=true
   )
  ]
  rendererClasses [&NiagaraRibbonRendererProperties]
  renderers [{rendererClass=&NiagaraRibbonRendererProperties}]
 }
}
summary {
 systemName=NS_Pickaxe_IcicleCane_Bright_Trail
 emitters [
  {emitterName=MainTrail, rendererClasses [&NiagaraRibbonRendererProperties]}
  {emitterName=Particle, rendererClasses [&NiagaraSpriteRendererProperties, &NiagaraMeshRendererProperties]}
  {emitterName=TipTrail, rendererClasses [&NiagaraRibbonRendererProperties]}
 ]
 userVariables [
  User.bIsFrontEnd {type=NiagaraBool}
  User.bIsFrontEndPreview {type=NiagaraBool}
  User.RibbonIDOffset {type=NiagaraInt32}
  User.SwingSectionIndex {type=NiagaraInt32}
  User.Variant {type=NiagaraFloat, defaultValue=0, description=""}
 ]
}
