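# Niagara effect for the jab of the Bright Icicle Cane pickaxe, a short icy-blue burst.
# Airflow adds mesh and sprite streaks, Soulercoasters adds rotating meshes with a platform-scaled count, and Particle launches delayed cone sprays with drag, gravity, curl noise and flutter.
# Spawn groups select per-burst settings and feed renderer visibility tags.
# Variant selectors pick colors, while time-of-day and front-end scalars adjust intensity and particle size.
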NiagaraSystem /BRCosmetics/Weapons/FORT_Melee/Pickaxe_IcicleCane_Bright/FX/NS_Pickaxe_IcicleCane_Jab
dependencies {
 usedDataInterfaces [
  &NiagaraDataInterfaceCurve
  &NiagaraDataInterfaceColorCurve
  &NiagaraDataInterfacePlatformSet
  &NiagaraDataInterfaceSpriteRendererInfo
 ]
 usedDynamicInputs [&FN_FrontEndFloatModifier]
 usedModules [
  &self:SystemSpawnScript.SystemScriptSource.SystemScriptGraph.NiagaraNodeAssignment_0.SetVariables_#1
  &self:SystemSpawnScript.SystemScriptSource.SystemScriptGraph.NiagaraNodeAssignment_1.SetVariables_#2
  &SystemState
  &FN_VariantManager
 ]
}
emitters {
 Airflow {
  emitterUpdateScript [
   &EmitterState()
   &SpawnBurst_Instantaneous(Use Spawn Probability=false, UseLoopCountLimit=false)
   &SpawnBurst_Instantaneous(Spawn Group=1, Use Spawn Probability=false, UseLoopCountLimit=false) {module=SpawnBurst_Instantaneous001}
  ]
  particleSpawnScript [
   &InitializeParticle(
    Color=<rgba 15 15 15 1>
    Lifetime=&MergeHelper_FloatSwitch(0=0.35, 1=0.15, 2=?, 3=?, Index=$Emitter.SpawnGroup, Num=::NewEnumerator0)
    Mesh Scale=<0.585 1 1>
    Mesh Scale Mode=::NewEnumerator4
    Position Offset=&MergeHelper_VectorSwitch(0=<0 0 100>, 1=<0 0 -15>, 2=?, 3=?, Index=$Emitter.SpawnGroup, Num=::NewEnumerator0)
    Sprite Size=<35 100>
    Sprite Size Mode=::NewEnumerator4
    UsePositionOffset=true
    Write Material Random=true
   )
   &AddVelocity(Velocity=&MakeVector(
     X=0
     Y=0
     Z=&MergeHelper_FloatSwitch(0=50, 1=635, 2=?, 3=?, Index=$Emitter.SpawnGroup, Num=::NewEnumerator0)
    ))
   &InitialMeshOrientation(Mesh Orientation Mode=::NewEnumerator2, Rotation=<0 -0.25 0>, Use Rotation Vector=true)
   &CameraOffset(Camera Offset Amount=&MergeHelper_FloatSwitch(0=20, 1=5, 2=?, 3=?, Index=$Emitter.SpawnGroup, Num=::NewEnumerator0))
   /SetVariables(Particles.VisibilityTag=$Emitter.SpawnGroup) {isSetParameters=true}
  ]
  particleUpdateScript [
   &ParticleState()
   &ScaleMeshSize(Scale Factor=&MakeVector(
     X=&FloatFromCurve(FloatCurve=DataInterface() {propertyValues {Curve {keys [<at 0 1.5>,<at 1 1>]}}})
     Y=1
     Z=1
    ))
   &SolveForcesAndVelocity(Clamp Velocity=false, Limit Acceleration=false)
   &DynamicMaterialParameters(
    Index 0 Param 1=&FloatFromCurve(FloatCurve=DataInterface() {propertyValues {Curve {keys [<at 0 0>,<at 0.5 0>,<at 1 0.08>]}}})
    Index 0 Param 2=&FloatFromCurve(FloatCurve=DataInterface() {propertyValues {Curve {keys [{arriveTangent=-2.2616, interpMode=RCIM_Cubic, time=0, value=-0.5}, {interpMode=RCIM_Cubic, leaveTangent=-0.7156, time=1, value=1}]}}})
    Index 0 Param 3=1.2
    Index 0 Param 4=&FloatFromCurve(FloatCurve=DataInterface() {propertyValues {Curve {keys [<at 0 1>,<at 1 2>]}}}, Scale Curve=0.2)
    Module.Index0Float0Write=true
    Module.Index0Float1Write=true
    Module.Index0Float2Write=true
    Module.Index0Float3Write=true
   )
   &ScaleColor(
    Scale Alpha=&Multiply_Float(
     A=&FloatFromCurve(
      FloatCurve=DataInterface() {propertyValues {Curve {keys [<at 0 1 RCIM_Cubic>,<at 1 0 RCIM_Cubic>]}}}
      Scale Curve=&Multiply_Float(A=0.5, B=$System.ToDScalar)
     )
     B=&FloatFromCurve(FloatCurve=DataInterface() {propertyValues {Curve {keys [<at 0 0 RCIM_Cubic>,<at 0.175 1 RCIM_Cubic>]}}}, CurveIndex=$Engine.System.Age)
    )
    Scale RGB=&MakeVectorFromLinearColorRGB(Color=&VariantColorSelector(
      Variant Amount=0
      Original Color=&ColorFromCurve(ColorCurve=DataInterface() {propertyValues {
         AlphaCurve {keys [<at 0 1 RCIM_Cubic>]}
         BlueCurve {keys [<at 0 4 RCIM_Cubic>,<at 1 1 RCIM_Cubic>]}
         GreenCurve {keys [<at 0 1.9 RCIM_Cubic>,<at 1 0.42 RCIM_Cubic>]}
         RedCurve {keys [<at 0 1.2 RCIM_Cubic>,<at 1 0.4 RCIM_Cubic>]}
        }})
      Variant 1 Color=&ColorFromCurve(ColorCurve=DataInterface() {propertyValues {
         AlphaCurve {keys [<at 0 1 RCIM_Cubic>]}
         BlueCurve {keys [<at 0 4 RCIM_Cubic>,<at 1 1 RCIM_Cubic>]}
         GreenCurve {keys [<at 0 2.5067 RCIM_Cubic>,<at 1 0.5829 RCIM_Cubic>]}
         RedCurve {keys [<at 0 0.8 RCIM_Cubic>,<at 1 0.285 RCIM_Cubic>]}
        }})
      Variant 2 Color=?
     ))
    ScaleA=true
    ScaleRGB=true
   )
   &FN_Sprite_Attributes(
    Color=&MakeLinearColorFromVectorAndFloat(
     "Vector (RGB)"=&Multiply_VectorByFloat(
      Vector=&MakeVectorFromLinearColorRGB(Color=&VariantColorSelector(
        Variant Amount=0
        Original Color=&ColorFromCurve(ColorCurve=DataInterface() {propertyValues {
           AlphaCurve {keys [<at 0 1 RCIM_Cubic>]}
           BlueCurve {keys [<at 0 2 RCIM_Cubic>,<at 1 4 RCIM_Cubic>]}
           GreenCurve {keys [<at 0 1.2 RCIM_Cubic>,<at 1 1.2 RCIM_Cubic>]}
           RedCurve {keys [<at 0 0.8 RCIM_Cubic>,<at 1 1.2 RCIM_Cubic>]}
          }})
        Variant 1 Color=&ColorFromCurve(ColorCurve=DataInterface() {propertyValues {
           AlphaCurve {keys [<at 0 1 RCIM_Cubic>]}
           BlueCurve {keys [<at 0 2 RCIM_Cubic>,<at 1 4 RCIM_Cubic>]}
           GreenCurve {keys [<at 0 1.2533 RCIM_Cubic>,<at 1 2.3083 RCIM_Cubic>]}
           RedCurve {keys [<at 0 0.4 RCIM_Cubic>,<at 1 1.1 RCIM_Cubic>]}
          }})
        Variant 2 Color=?
       ))
      Float=$System.ToDScalar
     )
     "Float (Alpha)"=&FloatFromCurve(FloatCurve=DataInterface() {propertyValues {Curve {keys [<at 0 0>,<at 0.1 1>,<at 0.7 1>,<at 1 0>]}}})
    )
    Create Color Attribute=true
    Write to Module Namespace=true
   )
  ]
  rendererClasses [&NiagaraMeshRendererProperties, &NiagaraSpriteRendererProperties]
  renderers [
   {rendererClass=&NiagaraMeshRendererProperties, rendererIndex=0}
   {rendererClass=&NiagaraSpriteRendererProperties, rendererIndex=1}
  ]
 }
 Particle {
  emitterUpdateScript [
   &EmitterState(Life Cycle Mode=::NewEnumerator1, Loop Behavior=::NewEnumerator1, Loop Delay=0.08, Loop Duration=0.225, UseLoopDelay=true)
   &SpawnBurst_Instantaneous(
    Spawn Count=&RandomRangeInt(Minimum=3, Maximum=4)
    Use Spawn Probability=false
    UseLoopCountLimit=false
   ) {module=SpawnBurst_Instantaneous002}
   &SpawnBurst_Instantaneous(
    Spawn Count=&RandomRangeInt(Minimum=2, Maximum=3)
    Spawn Group=1
    Use Spawn Probability=false
    UseLoopCountLimit=false
   )
   &SpawnBurst_Instantaneous(
    Spawn Count=&RandomRangeInt(Minimum=5, Maximum=6)
    Spawn Group=2
    Use Spawn Probability=false
    UseLoopCountLimit=false
   ) {module=SpawnBurst_Instantaneous001}
   /SetVariables(
    Emitter.PetalInnerColor:LinearColor=&VariantColorSelector(Variant Amount=0, Original Color=<rgba 0.4 0.55 1 1>, Variant 1 Color=<rgba 0.1878 0.5113 0.8348 1>, Variant 2 Color=?)
    Emitter.PetalMiddleColor:LinearColor=&VariantColorSelector(Variant Amount=0, Original Color=<rgba 0.65 0.9125 1 1>, Variant 1 Color=<rgba 0.3043 0.7913 1 1>, Variant 2 Color=?)
    Emitter.PetalOuterColor:LinearColor=&VariantColorSelector(Variant Amount=0, Original Color=<rgba 0.3583 0.3 1 1>, Variant 1 Color=<rgba 0.4 0.4 1 1>, Variant 2 Color=?)
   ) {isSetParameters=true}
  ]
  particleSpawnScript [
   &InitializeParticle(
    Color=&VariantColorSelector(Variant Amount=0, Original Color=<rgba 0.7 0.925 1 1>, Variant 1 Color=<rgba 0.4035 0.6134 0.807 1>, Variant 2 Color=?)
    Lifetime=&MergeHelper_FloatSwitch(
     0=&RandomRangeFloat(Minimum=0.7, Maximum=1)
     1=&RandomRangeFloat(Minimum=0.4, Maximum=0.65)
     2=&RandomRangeFloat(Minimum=0.45, Maximum=0.7)
     3=?
     Index=$Emitter.SpawnGroup
     Num=::NewEnumerator1
    )
    Mesh Scale Mode=::NewEnumerator3
    Mesh Uniform Scale=&Multiply_Float(A=&RandomRangeFloat(Minimum=0.1, Maximum=0.18), B=$System.FrontEndScale)
    Sprite Rotation Mode=::NewEnumerator2
    Sprite Size=&Multiply_Vector2D_ByFloat(
     Float=&Multiply_Float(A=&RandomRangeFloat(Minimum=0.245, Maximum=0.695), B=$System.FrontEndScale)
     Vector2D=<18.5 7>
    )
    Sprite Size Mode=::NewEnumerator4
    Sprite UV Mode=::NewEnumerator3
    UsePositionOffset=false
    Write Material Random=true
   )
   &ShapeLocation(
    MaskSpawnGroup=false
    Offset=<0 30 45>
    Offset Coordinate Space=::Local
    Offset Mode=::NewEnumerator0
    Override Seed=false
    Sphere Radius=&MergeHelper_FloatSwitch(0=15, 1=8, 2=3, 3=?, Index=$Emitter.SpawnGroup, Num=::NewEnumerator1)
    UseOwnerScale=true
   )
   &AddVelocity(
    Cone Angle=&MergeHelper_FloatSwitch(0=10, 1=15, 2=0, 3=?, Index=$Emitter.SpawnGroup, Num=::NewEnumerator1)
    Cone Axis=<0.3 0.3 1>
    Module.UseConeFalloff=false
    Module.UseVelDistribution=false
    Override Seed=false
    Velocity Mode=::NewEnumerator2
    Velocity Speed=&Multiply_Float(
     A=&MergeHelper_FloatSwitch(
      0=&RandomRangeFloat(Minimum=1000, Maximum=775)
      1=&RandomRangeFloat(Minimum=585, Maximum=685)
      2=&RandomRangeFloat(Minimum=750, Maximum=875)
      3=?
      Index=$Emitter.SpawnGroup
      Num=::NewEnumerator1
     )
     B=1.2
    )
   )
   &CameraOffset(Camera Offset Amount=&MergeHelper_FloatSwitch(0=20, 1=10, 2=20, 3=?, Index=$Emitter.SpawnGroup, Num=::NewEnumerator1))
   /SetVariables(Particles.VisibilityTag=$Emitter.SpawnGroup) {isSetParameters=true}
  ]
  particleUpdateScript [
   &ParticleState()
   &ScaleMeshSize(Scale Factor=&VectorFromFloat(Value=&FloatFromCurve(FloatCurve=DataInterface() {propertyValues {Curve {keys [<at 0 1 RCIM_Cubic>,<at 1 0 RCIM_Cubic>]}}}, Scale Curve=2)))
   &Drag(
    Drag=&Multiply_Float(
     A=&UniformRangedFloat(Minimum=1.2, Maximum=1.5)
     B=&FloatFromCurve(
      FloatCurve=DataInterface() {propertyValues {Curve {keys [
          {arriveTangent=-5.3551, interpMode=RCIM_Cubic, leaveTangent=-5.3551, tangentMode=RCTM_User, time=0, value=0}
          {arriveTangent=-138.8873, interpMode=RCIM_Cubic, leaveTangent=-138.8873, tangentMode=RCTM_User, time=0.05, value=7}
          <at 0.2 0.5>
         ]}}}
      Scale Curve=4
     )
    )
    Use Linear Drag=true
    Use Rotational Drag=false
   )
   &SubUVAnimation(
    Output Alternate Renderer Binding=true
    "Sprite Renderer "=DataInterface() {propertyValues {SpriteRenderer=&self:Particle_0.NiagaraSpriteRendererProperties_1}}
    UseEndFrame=false
    UseStartFrame=false
   )
   &ScaleColor(
    Scale RGB=&VectorFromFloat(Value=&FloatFromCurve(
      FloatCurve=DataInterface() {propertyValues {Curve {keys [{arriveTangent=-2.7905, interpMode=RCIM_Cubic, leaveTangent=-2.7905, tangentMode=RCTM_User, time=0, value=2}, <at 1 1 RCIM_Cubic>]}}}
      Scale Curve=$System.ToDScalar
     ))
    ScaleA=true
    ScaleRGB=true
   )
   &GravityForce(Gravity=&RandomRangeVector(Minimum=<0 0 -150>, Maximum=<0 0 -200>))
   &CurlNoiseForce(
    Bias Noise=false
    Noise Frequency=20
    Noise Strength=&MergeHelper_FloatSwitch(
     0=&RandomRangeFloat(Minimum=225, Maximum=350)
     1=0
     2=0
     3=?
     Index=$Particles.VisibilityTag
     Num=::NewEnumerator1
    )
    Pan Noise=false
   )
   &ScaleSpriteSize(Scale Factor=&Lerp_Vector2(
     A=&Multiply_Vector2(
      A=&MakeVector2D(
       X=&Multiply_Float(
        A=1
        B=&Sine(
         Normalized Angle=$Emitter.Age
         Period=&Multiply_Float(A=0.7, B=&Add_Float(A=$Particles.Initial.MaterialRandom, B=0.65))
         Scale=1
         Bias=0
        )
       )
       Y=1
      )
      B=&Vector2DFromFloat(Value=&FloatFromCurve(FloatCurve=DataInterface() {propertyValues {Curve {keys [<at 0 1.1>,<at 0.75 1.1>,<at 1 0>]}}}))
     )
     B=<1 1>
     Alpha=&FloatFromCurve(FloatCurve=DataInterface() {propertyValues {Curve {keys [<at 0.0574 1>,<at 0.1545 0>]}}})
    ))
   &SpriteRotationRate(Rotation Rate=&Multiply_Float(
     A=&UniformAOrBFloat(A=-1, Evaluation Type=::NewEnumerator0, Recalculate Random Each Loop=true, Override Seed=false)
     B=&MergeHelper_FloatSwitch(
      0=&UniformRangedFloat(Minimum=50, Maximum=200)
      1=0
      2=0
      3=?
      Index=$Particles.VisibilityTag
      Num=::NewEnumerator1
     )
    ))
   &FN_Sprite_Attributes(
    Color=&MakeLinearColorFromVectorAndFloat(
     "Vector (RGB)"=&MakeVectorFromLinearColorRGB(Color=&RandomRangeLinearColor(Minimum=<rgba 0.2563 0.299 0.4271 0.6>, Maximum=<rgba 0.6 0.6667 1 0.8>))
     "Float (Alpha)"=&FloatFromCurve(
      FloatCurve=DataInterface() {propertyValues {Curve {keys [
          {arriveTangent=7.9556, interpMode=RCIM_Cubic, leaveTangent=7.9556, tangentMode=RCTM_User, time=0, value=0}
          {arriveTangentWeight=0.0667, interpMode=RCIM_Cubic, leaveTangentWeight=0.1, tangentMode=RCTM_User, time=0.2, value=1}
          {arriveTangentWeight=0.1601, interpMode=RCIM_Cubic, leaveTangentWeight=0.2668, tangentMode=RCTM_User, time=0.5, value=1}
          {arriveTangent=-3.5681, tangentMode=RCTM_User, time=1, value=0}
         ]}}}
      Scale Curve=&Multiply_Float(A=0.08, B=&FNTimeOfDayScalar(Night Scale=0.65, Morning Scale=0.675, Evening Scale=0.7))
     )
    )
    Create Color Attribute=true
    Create SpriteSize Attribute=true
    SpriteSize=&Multiply_Vector2D_ByFloat(
     Float=&MakeFloatFromVector2d(Vector 2D=&Vector2DFromFloat(Value=&FloatFromCurve(FloatCurve=DataInterface() {propertyValues {Curve {keys [
            {arriveTangent=0.4145, interpMode=RCIM_Cubic, leaveTangent=0.4145, tangentMode=RCTM_User, time=0, value=0.8}
            {arriveTangent=0.2495, interpMode=RCIM_Cubic, leaveTangent=0.2495, tangentMode=RCTM_User, time=0.3, value=0.9}
            {arriveTangentWeight=0.2333, interpMode=RCIM_Cubic, tangentMode=RCTM_User, time=1, value=1}
           ]}}})))
     Vector2D=&Multiply_Vector2D_ByFloat(
      Float=$System.FrontEndScale
      Vector2D=&Vector2DFromFloat(Value=&RandomRangeFloat(Minimum=70, Maximum=90))
     )
    )
    Write to Module Namespace=true
   )
   &SolveForcesAndVelocity(Clamp Velocity=false, Limit Acceleration=false)
  ]
  rendererClasses [&NiagaraSpriteRendererProperties, &NiagaraMeshRendererProperties]
  renderers [
   {rendererClass=&NiagaraSpriteRendererProperties}
   {rendererClass=&NiagaraMeshRendererProperties, rendererIndex=1}
   {rendererClass=&NiagaraSpriteRendererProperties, rendererIndex=2}
  ]
 }
 Soulercoasters {
  emitterUpdateScript [
   &EmitterState()
   &SpawnBurst_Instantaneous(
    Spawn Count=&MakeIntFromFloat(Float=&FN_PickFloatDependOnPlatform(TargetPlatform=DataInterface() {propertyValues {Platforms {qualityLevelMask=-29}}}, TargetPlatformActive=3, TargetPlatformNoActive=5))
    Use Spawn Probability=false
    UseLoopCountLimit=false
   )
   &SpawnBurst_Instantaneous(Spawn Group=1, Spawn Time=0.085, Use Spawn Probability=false, UseLoopCountLimit=false) {module=SpawnBurst_Instantaneous001}
  ]
  particleSpawnScript [
   &InitializeParticle(
    Color=<rgba 5 5 5 1>
    Lifetime=&MergeHelper_FloatSwitch(0=0.25, 1=0.2, 2=?, 3=?, Index=$Emitter.SpawnGroup, Num=::NewEnumerator0)
    Mesh Scale=&MergeHelper_VectorSwitch(
     0=&MakeVector(X=1, Y=1, Z=&RandomRangeFloat(Minimum=1.1, Maximum=1.5))
     1=<0.1 0.075 0.075>
     2=?
     3=?
     Index=$Emitter.SpawnGroup
     Num=::NewEnumerator0
    )
    Mesh Scale Mode=::NewEnumerator4
    Position Offset=&MergeHelper_VectorSwitch(0=<0 0 110>, 1=<0 0 100>, 2=?, 3=?, Index=$Emitter.SpawnGroup, Num=::NewEnumerator0)
    UsePositionOffset=true
    Write Material Random=true
   )
   &AddRotationalVelocity(Rotation Rate=<0 0 -0.125>)
   &InitialMeshOrientation(
    Mesh Orientation Mode=::NewEnumerator7
    Rotation=&MergeHelper_VectorSwitch(
     0=&MakeVector(X=0, Y=0, Z=&ReturnNormalizedExecIndex(Normalized Index Includes One=false))
     1=<0 -0.25 0>
     2=?
     3=?
     Index=$Emitter.SpawnGroup
     Num=::NewEnumerator0
    )
    Use Rotation Vector=true
   )
   /SetVariables(Particles.VisibilityTag=$Emitter.SpawnGroup) {isSetParameters=true}
   &CameraOffset(Camera Offset Amount=20)
  ]
  particleUpdateScript [
   &ParticleState()
   &ScaleMeshSize(Scale Factor=&MergeHelper_VectorSwitch(0=<1 1 1>, 1=<1 1 1>, 2=?, 3=?, Index=$Particles.VisibilityTag, Num=::NewEnumerator0))
   &ScaleColor(
    Scale Alpha=&FloatFromCurve(
     FloatCurve=DataInterface() {propertyValues {Curve {keys [{arriveTangent=-2.9154, interpMode=RCIM_Cubic, leaveTangent=-2.9154, tangentMode=RCTM_User, time=0, value=1}, <at 1 0>]}}}
     Scale Curve=$System.ToDScalar
    )
    Scale RGB=&MergeHelper_VectorSwitch(
     0=&MakeVectorFromLinearColorRGB(Color=&VariantColorSelector(
       Variant Amount=0
       Original Color=&ColorFromCurve(ColorCurve=DataInterface() {propertyValues {
          AlphaCurve {keys [{interpMode=RCIM_Cubic, tangentMode=RCTM_User, time=0, value=1}]}
          BlueCurve {keys [
            {arriveTangentWeight=0.0472, interpMode=RCIM_Cubic, leaveTangentWeight=0.0472, tangentMode=RCTM_User, time=0.0259, value=1}
            {tangentMode=RCTM_User, time=0.1676, value=1}
            {arriveTangentWeight=0.2775, interpMode=RCIM_Cubic, leaveTangentWeight=0.2775, tangentMode=RCTM_User, time=0.7427, value=1}
           ]}
          GreenCurve {keys [
            {arriveTangentWeight=0.0472, interpMode=RCIM_Cubic, leaveTangentWeight=0.0472, tangentMode=RCTM_User, time=0.0259, value=0.475}
            {tangentMode=RCTM_User, time=0.1676, value=0.4167}
            {arriveTangentWeight=0.2775, interpMode=RCIM_Cubic, leaveTangentWeight=0.2775, tangentMode=RCTM_User, time=0.7427, value=0.3}
           ]}
          RedCurve {keys [
            {arriveTangentWeight=0.0472, interpMode=RCIM_Cubic, leaveTangentWeight=0.0472, tangentMode=RCTM_User, time=0.0259, value=0.3}
            {tangentMode=RCTM_User, time=0.1676, value=0.3}
            {arriveTangentWeight=0.2775, interpMode=RCIM_Cubic, leaveTangentWeight=0.2775, tangentMode=RCTM_User, time=0.7427, value=0.3}
           ]}
         }})
       Variant 1 Color=&ColorFromCurve(ColorCurve=DataInterface() {propertyValues {
          AlphaCurve {keys [{interpMode=RCIM_Cubic, tangentMode=RCTM_User, time=0, value=1}]}
          BlueCurve {keys [
            {arriveTangentWeight=0.0472, interpMode=RCIM_Cubic, leaveTangentWeight=0.0472, tangentMode=RCTM_User, time=0.0259, value=1}
            {tangentMode=RCTM_User, time=0.1676, value=1}
            {arriveTangentWeight=0.2775, interpMode=RCIM_Cubic, leaveTangentWeight=0.2775, tangentMode=RCTM_User, time=0.7427, value=1}
           ]}
          GreenCurve {keys [
            {arriveTangentWeight=0.0472, interpMode=RCIM_Cubic, leaveTangentWeight=0.0472, tangentMode=RCTM_User, time=0.0259, value=0.6267}
            {tangentMode=RCTM_User, time=0.1676, value=0.6125}
            {arriveTangentWeight=0.2775, interpMode=RCIM_Cubic, leaveTangentWeight=0.2775, tangentMode=RCTM_User, time=0.7427, value=0.5829}
           ]}
          RedCurve {keys [
            {arriveTangentWeight=0.0472, interpMode=RCIM_Cubic, leaveTangentWeight=0.0472, tangentMode=RCTM_User, time=0.0259, value=0.2}
            {tangentMode=RCTM_User, time=0.1676, value=0.225}
            {arriveTangentWeight=0.2775, interpMode=RCIM_Cubic, leaveTangentWeight=0.2775, tangentMode=RCTM_User, time=0.7427, value=0.285}
           ]}
         }})
       Variant 2 Color=?
      ))
     1=&MakeVectorFromLinearColorRGB(Color=&VariantColorSelector(
       Variant Amount=0
       Original Color=&ColorFromCurve(ColorCurve=DataInterface() {propertyValues {
          AlphaCurve {keys [<at 0 1 RCIM_Cubic>]}
          BlueCurve {keys [<at 0.1878 1.5>, {arriveTangent=-1.1081, interpMode=RCIM_Cubic, leaveTangent=-1.1081, time=0.6331, value=1.35}, <at 1 0.6 RCIM_Cubic>]}
          GreenCurve {keys [<at 0.1878 0.7125>, {arriveTangent=-0.6556, interpMode=RCIM_Cubic, leaveTangent=-0.6556, time=0.6331, value=0.4169}, <at 1 0.18 RCIM_Cubic>]}
          RedCurve {keys [<at 0.1878 0.45>, {arriveTangent=-0.3324, interpMode=RCIM_Cubic, leaveTangent=-0.3324, time=0.6331, value=0.3848}, <at 1 0.18 RCIM_Cubic>]}
         }})
       Variant 1 Color=&ColorFromCurve(ColorCurve=DataInterface() {propertyValues {
          AlphaCurve {keys [<at 0 1 RCIM_Cubic>]}
          BlueCurve {keys [<at 0.1878 1.5>, {arriveTangent=-1.1081, interpMode=RCIM_Cubic, leaveTangent=-1.1081, time=0.6331, value=1.35}, <at 1 0.6 RCIM_Cubic>]}
          GreenCurve {keys [<at 0.1878 0.94>, {arriveTangent=-0.7418, interpMode=RCIM_Cubic, leaveTangent=-0.7418, time=0.6331, value=0.8674}, <at 1 0.3375 RCIM_Cubic>]}
          RedCurve {keys [<at 0.1878 0.3>, {arriveTangent=-0.1847, interpMode=RCIM_Cubic, leaveTangent=-0.1847, time=0.6331, value=0.3848}, <at 1 0.15 RCIM_Cubic>]}
         }})
       Variant 2 Color=?
      ))
     2=?
     3=?
     Index=$Particles.VisibilityTag
     Num=::NewEnumerator0
    )
    ScaleA=true
    ScaleRGB=true
   )
   &DynamicMaterialParameters(
    Index 0 Param 1=&FloatFromCurve(FloatCurve=DataInterface() {propertyValues {Curve {keys [{arriveTangent=2.8243, interpMode=RCIM_Cubic, leaveTangent=2.8243, tangentMode=RCTM_User, time=0, value=-0.872}, <at 1 0.7557>]}}})
    Index 0 Param 2=1
    Index 0 Param 4=1
    Module.Index0Float0Write=true
    Module.Index0Float1Write=true
    Module.Index0Float2Write=true
    Module.Index0Float3Write=true
   )
   /SetVariables(Particles.DynamicParametersAlt:Vector4f=&MakeVector4(
     X=&FloatFromCurve(FloatCurve=DataInterface() {propertyValues {Curve {keys [<at 0 0>,<at 1 0.3046>]}}})
     Y=1
     Z=0
     W=1
    )) {isSetParameters=true}
   &SolveForcesAndVelocity(Clamp Velocity=false, Limit Acceleration=false)
  ]
  rendererClasses [&NiagaraMeshRendererProperties]
  renderers [
   {rendererClass=&NiagaraMeshRendererProperties, rendererIndex=0}
   {rendererClass=&NiagaraMeshRendererProperties, rendererIndex=1}
  ]
 }
}
summary {
 systemName=NS_Pickaxe_IcicleCane_Jab
 emitters [
  {emitterName=Airflow, rendererClasses [&NiagaraMeshRendererProperties, &NiagaraSpriteRendererProperties]}
  {emitterName=Soulercoasters, rendererClasses [&NiagaraMeshRendererProperties]}
  {emitterName=Particle, rendererClasses [&NiagaraSpriteRendererProperties, &NiagaraMeshRendererProperties]}
 ]
 userVariables [
  User.bIsFrontEnd {type=NiagaraBool}
  User.bIsFrontEndPreview {type=NiagaraBool}
  User.SwingSectionIndex {type=NiagaraInt32}
  User.Variant {type=NiagaraFloat, defaultValue=0, description=""}
 ]
}
